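; CacaoLynx traversal anim notify state that hides the player's companion while the traversal animation plays.
; Begin re-enables move input, sets walk speed, hides the first "PlayerCompanion"-tagged actor's mesh, plays its hide montage and disables collision.
; Locally controlled Tick monitors input and movement, ending the contextual scene with a reappear montage on non-forward input, firing, sprinting, crouching, water or falling.
; End restores mesh, Niagara visibility and collision.
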(fn Received_NotifyBegin
  (params (MeshComp :type Skeletal%20Mesh%20Component%20Object%20Reference) (Animation :type Anim%20Sequence%20Base%20Object%20Reference) (TotalDuration :type Float%20%28single-precision%29) (EventReference :type Anim%20Notify%20Event%20Reference%20Structure%20%28by%20ref%29))
  (bind _asfort_pawn (Utilities|Casting|CastToFortPawn (Components|GetOwner MeshComp)))
  (bind _asplayer_controller (Utilities|Casting|CastToPlayerController (Pawn|GetController _asfort_pawn))
    (:then
      (Input|SetIgnoreMoveInput _asplayer_controller false))
    (:CastFailed
      _))
  (bind _outactors (Actor|GetAllActorswithTag _asfort_pawn "PlayerCompanion"))
  (Rendering|SetVisibility (Actor|GetComponentbyClass (Utilities|Array|Get(acopy) _outactors) "/Script/Engine.SkeletalMeshComponent") false false)
  (Variables|/Script/Engine.CharacterMovementComponent|SetMaxWalkSpeed 5.0 (Variables|/Script/Engine.Character|GetCharacterMovement _asfort_pawn))
  (Animation|Montage|MontagePlay (Components|SkeletalMesh|GetAnimInstance (Actor|GetComponentbyClass (Utilities|Array|Get(acopy) _outactors) "/Script/Engine.SkeletalMeshComponent")) "/CosmeticCompanions/Assets/Quadruped/CacaoLynx/Animation/Emotes/CacaoLynx_Emote_hide_M.CacaoLynx_Emote_hide_M" 1.0 "MontageLength" 0.0 true) ; hide montage
  (ContextualAnim|SceneActorComponent|EarlyOutContextualAnimScene (Actor|GetComponentbyClass (Utilities|Array|Get(acopy) _outactors) "/Script/ContextualAnimation.ContextualAnimSceneActorComponent") false)
  (Collision|SetActorEnableCollision (Utilities|Array|Get(acopy) _outactors) false)
  (return false))

(fn Received_NotifyEnd
  (params (MeshComp :type Skeletal%20Mesh%20Component%20Object%20Reference) (Animation :type Anim%20Sequence%20Base%20Object%20Reference) (EventReference :type Anim%20Notify%20Event%20Reference%20Structure%20%28by%20ref%29))
  (bind _asfort_pawn (Utilities|Casting|CastToFortPawn (Components|GetOwner MeshComp)))
  (bind _outactors (Actor|GetAllActorswithTag _asfort_pawn "PlayerCompanion"))
  (Rendering|SetVisibility (Actor|GetComponentbyClass (Utilities|Array|Get(acopy) _outactors) "/Script/Engine.SkeletalMeshComponent") true false)
  (Rendering|SetVisibility (Actor|GetComponentbyClass (Utilities|Array|Get(acopy) _outactors) "/Script/Niagara.NiagaraComponent") true false)
  (Animation|Montage|MontageStop (Components|SkeletalMesh|GetAnimInstance (Actor|GetComponentbyClass (Utilities|Array|Get(acopy) _outactors) "/Script/Engine.SkeletalMeshComponent")) 0.0 "/CosmeticCompanions/Assets/Quadruped/CacaoLynx/Animation/Emotes/CacaoLynx_Emote_hide_M.CacaoLynx_Emote_hide_M") ; hide montage
  (Collision|SetActorEnableCollision (Utilities|Array|Get(acopy) _outactors) true)
  (return false))

(fn Received_NotifyTick
  (params (MeshComp :type Skeletal%20Mesh%20Component%20Object%20Reference) (Animation :type Anim%20Sequence%20Base%20Object%20Reference) (FrameDeltaTime :type Float%20%28single-precision%29) (EventReference :type Anim%20Notify%20Event%20Reference%20Structure%20%28by%20ref%29))
  (bind _returnvalue (Components|GetOwner MeshComp))
  (bind _asfort_pawn (Utilities|Casting|CastToFortPawn _returnvalue))
  (bind _asfort_player_pawn (Utilities|Casting|CastToFortPlayerPawn _returnvalue))
  (if (Pawn|IsLocallyControlled _asfort_pawn)
    (if (Variables|/Script/FortniteGame.FortPawn|GetbPrimaryInputHeld _asfort_pawn)
      _
      (elif (Variables|/Script/FortniteGame.FortPlayerPawn|GetbIsTacticalSprinting _asfort_player_pawn)
        _
        (elif (Variables|/Script/Engine.Character|GetbIsCrouched _asfort_pawn)
          _
          (else
            (bind _isinwater (InterfacePlayerPawn|GetinWater(Message) _asfort_pawn))
            (if _isinwater
              _
              (elif (Movement|IsFalling (Variables|/Script/Engine.Character|GetCharacterMovement _asfort_pawn))
                _
                (elif (< (Math|Vector|VectorLength (Pawn|Components|PawnMovement|GetLastInputVector (Variables|/Script/Engine.Character|GetCharacterMovement _asfort_pawn))) 0.01)
                  _
                  (elif (> (Math|Vector|DotProduct (Transformation|GetActorForwardVector _asfort_pawn) (Math|Vector|Normalize (Pawn|Components|PawnMovement|GetLastInputVector (Variables|/Script/Engine.Character|GetCharacterMovement _asfort_pawn)))) 0.75)
                    (Development|PrintString 0 (Utilities|String|ToString(Float) (Math|Vector|DotProduct (Transformation|GetActorForwardVector _asfort_pawn) (Math|Vector|Normalize (Pawn|Components|PawnMovement|GetLastInputVector (Variables|/Script/Engine.Character|GetCharacterMovement _asfort_pawn))))))
                    ; hide spark mode
                    (bind _outactors (Actor|GetAllActorswithTag _asfort_pawn "PlayerCompanion"))
                    (Rendering|SetVisibility (Actor|GetComponentbyClass (Utilities|Array|Get(acopy) _outactors) "/Script/Niagara.NiagaraComponent") false false)))))))))
    ; Ends montage when any input other than forward is detected, or player enters water, falls, fires, sprints
    (bind _outactors_2 (Actor|GetAllActorswithTag _asfort_pawn "PlayerCompanion"))
    (Animation|Montage|MontageStop (Components|SkeletalMesh|GetAnimInstance (Actor|GetComponentbyClass (Utilities|Array|Get(acopy) _outactors_2) "/Script/Engine.SkeletalMeshComponent")) 0.0 "/CosmeticCompanions/Assets/Quadruped/CacaoLynx/Animation/Emotes/CacaoLynx_Emote_hide_M.CacaoLynx_Emote_hide_M") ; hide montage
    (Animation|Montage|MontagePlay (Components|SkeletalMesh|GetAnimInstance (Actor|GetComponentbyClass (Utilities|Array|Get(acopy) _outactors_2) "/Script/Engine.SkeletalMeshComponent")) "/CosmeticCompanions/Assets/Quadruped/CacaoLynx/Animation/Emotes/CacaoLynx_Emote_reappear_M.CacaoLynx_Emote_reappear_M" 1.0 "MontageLength" 0.0 true) ; reappear montage
    (Collision|SetActorEnableCollision (Utilities|Array|Get(acopy) _outactors_2) true)
    (ContextualAnim|SceneActorComponent|EarlyOutContextualAnimScene (Actor|GetComponentbyClass _asfort_pawn "/Script/ContextualAnimation.ContextualAnimSceneActorComponent") true)))
