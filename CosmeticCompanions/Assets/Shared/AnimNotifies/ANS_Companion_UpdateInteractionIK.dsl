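; Anim notify state that keeps a player's pose aligned with a companion during an interaction animation, using IK offsets.
; Each tick, only where the mesh owner lacks network authority, it calls UpdateInteractionIK on the companion's anim instance to get the instigator's anim instance and offsets.
; On success, it casts that instance to the Fortnite player AnimBlueprint and writes the IK offset to LeftHandOverride and the pelvis offset to EmoteAnimOffset.
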(fn Received_NotifyTick
  (params (MeshComp :type Skeletal%20Mesh%20Component%20Object%20Reference) (Animation :type Anim%20Sequence%20Base%20Object%20Reference) (FrameDeltaTime :type Float%20%28single-precision%29) (EventReference :type Anim%20Notify%20Event%20Reference%20Structure%20%28by%20ref%29))
  (if (not (Networking|HasAuthority (Components|GetOwner MeshComp)))
    (bind _asfort_player_companion_anim_instance (Utilities|Casting|CastToFortPlayerCompanionAnimInstance (Components|SkeletalMesh|GetAnimInstance MeshComp)))
    (bind (_bsuccess _instigatoraniminstance _instigatorikoffsetcs _instigatorpelvisoffsetcs) (Anim|Companion|UpdateInteractionIK _asfort_player_companion_anim_instance))
    (if _bsuccess
      (bind _asfortnite_m_avg_player_anim_blueprint (Utilities|Casting|CastToFortnite_M_Avg_Player_AnimBlueprint _instigatoraniminstance))
      (Variables|/Game/Characters/Player/Male/Male_Avg_Base/Fortnite_M_Avg_Player_AnimBlueprint.Fortnite_M_Avg_Player_AnimBlueprint_C|SetLeftHandOverride _instigatorikoffsetcs _asfortnite_m_avg_player_anim_blueprint)
      (Variables|/Script/FortniteGame.FortPlayerAnimInstance|SetEmoteAnimOffset _instigatorpelvisoffsetcs _asfortnite_m_avg_player_anim_blueprint)
      (return false))))
