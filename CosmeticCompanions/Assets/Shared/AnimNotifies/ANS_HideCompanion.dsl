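; Anim notify state that hides a cosmetic companion's body for the duration of the notify window.
; On begin, it finds the owner's skeletal mesh components with "Body" in their display name and hides them, propagating to children.
; On end, it shows them again unless an active Niagara component with "Spark" in its name is found on the owner, in which case visibility is left untouched.
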(fn Received_NotifyBegin
  (params (MeshComp :type Skeletal%20Mesh%20Component%20Object%20Reference) (Animation :type Anim%20Sequence%20Base%20Object%20Reference) (TotalDuration :type Float%20%28single-precision%29) (EventReference :type Anim%20Notify%20Event%20Reference%20Structure%20%28by%20ref%29))
  (for _array_element (Actor|GetComponentsByClass (Components|GetOwner MeshComp) "/Script/Engine.SkeletalMeshComponent")
    (if (Utilities|String|Contains (Utilities|GetDisplayName _array_element) "Body")
      (Rendering|SetVisibility _array_element false true)))
  (return false))

(fn Received_NotifyEnd
  (params (MeshComp :type Skeletal%20Mesh%20Component%20Object%20Reference) (Animation :type Anim%20Sequence%20Base%20Object%20Reference) (EventReference :type Anim%20Notify%20Event%20Reference%20Structure%20%28by%20ref%29))
  (bind _returnvalue (Components|GetOwner MeshComp))
  ; look for "Spark" NS and doesn't set vis if true
  (for _array_element (Actor|GetComponentsByClass _returnvalue "/Script/Niagara.NiagaraComponent")
    (if (Utilities|String|Contains (Utilities|GetDisplayName _array_element) "Spark")
      (if (Components|Activation|IsActive _array_element)
        (LocalVariables|SetbSparkActive true))))
  (if (not (LocalVariables|GetbSparkActive))
    (for _array_element_2 (Actor|GetComponentsByClass _returnvalue "/Script/Engine.SkeletalMeshComponent")
      (if (Utilities|String|Contains (Utilities|GetDisplayName _array_element_2) "Body")
        (Rendering|SetVisibility _array_element_2 true true))))
  (return false))
