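; Anim notify state for cosmetic companions that pins the animated mesh's owner to the ground in front of the player while the notify runs.
; On begin, it disables the owner's actor tick and line traces straight down from a point ahead of the player pawn.
; It moves both the owner and mesh component to the trace's hit location without checking for a hit.
; On end, it re-enables the owner's tick.
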(fn Received_NotifyBegin
  (params (MeshComp :type Skeletal%20Mesh%20Component%20Object%20Reference) (Animation :type Anim%20Sequence%20Base%20Object%20Reference) (TotalDuration :type Float%20%28single-precision%29) (EventReference :type Anim%20Notify%20Event%20Reference%20Structure%20%28by%20ref%29))
  (bind _returnvalue (Components|GetOwner MeshComp))
  (Actor|Tick|SetActorTickEnabled _returnvalue false)
  (bind _outhit (Collision|LineTraceByChannel _returnvalue (+ (Transformation|GetActorLocation (Game|GetPlayerPawn _returnvalue)) (* (Transformation|GetActorForwardVector (Game|GetPlayerPawn _returnvalue)) 200)) (Math|Vector|MakeVector (.x (+ (Transformation|GetActorLocation (Game|GetPlayerPawn _returnvalue)) (* (Transformation|GetActorForwardVector (Game|GetPlayerPawn _returnvalue)) 200))) (.y (+ (Transformation|GetActorLocation (Game|GetPlayerPawn _returnvalue)) (* (Transformation|GetActorForwardVector (Game|GetPlayerPawn _returnvalue)) 200))) (- (.z (+ (Transformation|GetActorLocation (Game|GetPlayerPawn _returnvalue)) (* (Transformation|GetActorForwardVector (Game|GetPlayerPawn _returnvalue)) 200))) 5e+02))))
  (Transformation|SetActorLocation _returnvalue (.Location (Collision|BreakHitResult _outhit)) false false)
  (Transformation|SetWorldLocation MeshComp (.Location (Collision|BreakHitResult _outhit)) false false)
  (return false))

(fn Received_NotifyEnd
  (params (MeshComp :type Skeletal%20Mesh%20Component%20Object%20Reference) (Animation :type Anim%20Sequence%20Base%20Object%20Reference) (EventReference :type Anim%20Notify%20Event%20Reference%20Structure%20%28by%20ref%29))
  (Actor|Tick|SetActorTickEnabled (Components|GetOwner MeshComp) true)
  (return false))
