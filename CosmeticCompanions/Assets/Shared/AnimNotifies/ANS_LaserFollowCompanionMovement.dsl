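; Anim notify state for the player's laser emote, making the first "PlayerCompanion"-tagged actor, a FloppyNib cat, chase a laser dot.
; Begin takes over companion behavior, disables collision, spawns the decal and ensures an emote component exists to hold all state.
; Tick traces from the camera, moves the dot and sets movement goals or triggers crouch, random swipes and pounces, toggling actor tick.
; End restores behavior, collision and tick, and destroys the decal.
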(fn Received_NotifyBegin
  (params (MeshComp :type Skeletal%20Mesh%20Component%20Object%20Reference) (Animation :type Anim%20Sequence%20Base%20Object%20Reference) (TotalDuration :type Float%20%28single-precision%29) (EventReference :type Anim%20Notify%20Event%20Reference%20Structure%20%28by%20ref%29))
  (bind _asfort_pawn (Utilities|Casting|CastToFortPawn (Components|GetOwner MeshComp)))
  (bind _outactors (Actor|GetAllActorswithTag _asfort_pawn "PlayerCompanion"))
  (bind _ascompanion_movement_interface (Utilities|Casting|CastToCompanionMovementInterface (Utilities|Array|Get(acopy) _outactors)))
  (Actor|Tick|SetActorTickEnabled (Utilities|Array|Get(acopy) _outactors) true)
  (if (not (Utilities|IsValid (Actor|GetComponentbyClass (Utilities|Array|Get(acopy) _outactors) "/CosmeticCompanions/Assets/Quadruped/FloppyNib/Animation/Emotes/FlobbyNib_LaserEmote/FloppyNib_Emote_Component.FloppyNib_Emote_Component_C") :!class /Script/Engine.KismetSystemLibrary))
    (Game|AddComponentbyClass (Utilities|Array|Get(acopy) _outactors) 0 "/CosmeticCompanions/Assets/Quadruped/FloppyNib/Animation/Emotes/FlobbyNib_LaserEmote/FloppyNib_Emote_Component.FloppyNib_Emote_Component_C"))
  (Companion|SetCustomBehavior _ascompanion_movement_interface true)
  (Variables|/CosmeticCompanions/Assets/Quadruped/FloppyNib/Animation/Emotes/FlobbyNib_LaserEmote/FloppyNib_Emote_Component.FloppyNib_Emote_Component_C|SetLaserStartLocation (Transformation|GetActorLocation _asfort_pawn) (Actor|GetComponentbyClass (Utilities|Array|Get(acopy) _outactors) "/CosmeticCompanions/Assets/Quadruped/FloppyNib/Animation/Emotes/FlobbyNib_LaserEmote/FloppyNib_Emote_Component.FloppyNib_Emote_Component_C"))
  (Collision|SetActorEnableCollision (Utilities|Array|Get(acopy) _outactors) false)
  (Animation|Montage|MontageStop (Components|SkeletalMesh|GetAnimInstance (Actor|GetComponentbyClass (Utilities|Array|Get(acopy) _outactors) "/Script/Engine.SkeletalMeshComponent")) 0.25 0)
  (Variables|/CosmeticCompanions/Assets/Quadruped/FloppyNib/Animation/Emotes/FlobbyNib_LaserEmote/FloppyNib_Emote_Component.FloppyNib_Emote_Component_C|SetbIsFollowingLaser true (Actor|GetComponentbyClass (Utilities|Array|Get(acopy) _outactors) "/CosmeticCompanions/Assets/Quadruped/FloppyNib/Animation/Emotes/FlobbyNib_LaserEmote/FloppyNib_Emote_Component.FloppyNib_Emote_Component_C"))
  (Variables|/CosmeticCompanions/Assets/Quadruped/FloppyNib/Animation/Emotes/FlobbyNib_LaserEmote/FloppyNib_Emote_Component.FloppyNib_Emote_Component_C|SetbShouldCrouch false (Actor|GetComponentbyClass (Utilities|Array|Get(acopy) _outactors) "/CosmeticCompanions/Assets/Quadruped/FloppyNib/Animation/Emotes/FlobbyNib_LaserEmote/FloppyNib_Emote_Component.FloppyNib_Emote_Component_C"))
  (Variables|/CosmeticCompanions/Assets/Quadruped/FloppyNib/Animation/Emotes/FlobbyNib_LaserEmote/FloppyNib_Emote_Component.FloppyNib_Emote_Component_C|SetbShouldPounce false (Actor|GetComponentbyClass (Utilities|Array|Get(acopy) _outactors) "/CosmeticCompanions/Assets/Quadruped/FloppyNib/Animation/Emotes/FlobbyNib_LaserEmote/FloppyNib_Emote_Component.FloppyNib_Emote_Component_C"))
  (Variables|/CosmeticCompanions/Assets/Quadruped/FloppyNib/Animation/Emotes/FlobbyNib_LaserEmote/FloppyNib_Emote_Component.FloppyNib_Emote_Component_C|SetCrouchTimer 0.0 (Actor|GetComponentbyClass (Utilities|Array|Get(acopy) _outactors) "/CosmeticCompanions/Assets/Quadruped/FloppyNib/Animation/Emotes/FlobbyNib_LaserEmote/FloppyNib_Emote_Component.FloppyNib_Emote_Component_C"))
  (Variables|/CosmeticCompanions/Assets/Quadruped/FloppyNib/Animation/Emotes/FlobbyNib_LaserEmote/FloppyNib_Emote_Component.FloppyNib_Emote_Component_C|SetPounceTimer 0.0 (Actor|GetComponentbyClass (Utilities|Array|Get(acopy) _outactors) "/CosmeticCompanions/Assets/Quadruped/FloppyNib/Animation/Emotes/FlobbyNib_LaserEmote/FloppyNib_Emote_Component.FloppyNib_Emote_Component_C"))
  (bind _returnvalue (Rendering|Decal|SpawnDecalatLocation (Utilities|Array|Get(acopy) _outactors) "/CosmeticCompanions/Assets/Quadruped/FloppyNib/Animation/Emotes/FlobbyNib_LaserEmote/FloppyNib_LaserEmote_Player/LaserDot.LaserDot" "4.000000,4.000000,4.000000" "0, 0, 0" "0.000000,0.000000,0.000000"))
  (bind _output_get (Variables|/CosmeticCompanions/Assets/Quadruped/FloppyNib/Animation/Emotes/FlobbyNib_LaserEmote/FloppyNib_Emote_Component.FloppyNib_Emote_Component_C|SetLaserDotDecal _returnvalue (Actor|GetComponentbyClass (Utilities|Array|Get(acopy) _outactors) "/CosmeticCompanions/Assets/Quadruped/FloppyNib/Animation/Emotes/FlobbyNib_LaserEmote/FloppyNib_Emote_Component.FloppyNib_Emote_Component_C")))
  (Rendering|Components|Decal|SetFadeScreenSize _output_get 0.001))

(fn Received_NotifyEnd
  (params (MeshComp :type Skeletal%20Mesh%20Component%20Object%20Reference) (Animation :type Anim%20Sequence%20Base%20Object%20Reference) (EventReference :type Anim%20Notify%20Event%20Reference%20Structure%20%28by%20ref%29))
  (bind _asfort_pawn (Utilities|Casting|CastToFortPawn (Components|GetOwner MeshComp)))
  (bind _outactors (Actor|GetAllActorswithTag _asfort_pawn "PlayerCompanion"))
  (Companion|SetCustomBehavior(Message) (Utilities|Array|Get(acopy) _outactors) false)
  (Collision|SetActorEnableCollision (Utilities|Array|Get(acopy) _outactors) true)
  (Actor|Tick|SetActorTickEnabled (Utilities|Array|Get(acopy) _outactors) true)
  (Variables|/CosmeticCompanions/Assets/Quadruped/FloppyNib/Animation/Emotes/FlobbyNib_LaserEmote/FloppyNib_Emote_Component.FloppyNib_Emote_Component_C|SetbIsFollowingLaser false (Actor|GetComponentbyClass (Utilities|Array|Get(acopy) _outactors) "/CosmeticCompanions/Assets/Quadruped/FloppyNib/Animation/Emotes/FlobbyNib_LaserEmote/FloppyNib_Emote_Component.FloppyNib_Emote_Component_C"))
  (Components|DestroyComponent (Variables|/CosmeticCompanions/Assets/Quadruped/FloppyNib/Animation/Emotes/FlobbyNib_LaserEmote/FloppyNib_Emote_Component.FloppyNib_Emote_Component_C|GetLaserDotDecal (Actor|GetComponentbyClass (Utilities|Array|Get(acopy) _outactors) "/CosmeticCompanions/Assets/Quadruped/FloppyNib/Animation/Emotes/FlobbyNib_LaserEmote/FloppyNib_Emote_Component.FloppyNib_Emote_Component_C")) 0)
  (Variables|/CosmeticCompanions/Assets/Quadruped/FloppyNib/Animation/Emotes/FlobbyNib_LaserEmote/FloppyNib_Emote_Component.FloppyNib_Emote_Component_C|SetLaserDotDecal 0 (Actor|GetComponentbyClass (Utilities|Array|Get(acopy) _outactors) "/CosmeticCompanions/Assets/Quadruped/FloppyNib/Animation/Emotes/FlobbyNib_LaserEmote/FloppyNib_Emote_Component.FloppyNib_Emote_Component_C")))

(fn Received_NotifyTick
  (params (MeshComp :type Skeletal%20Mesh%20Component%20Object%20Reference) (Animation :type Anim%20Sequence%20Base%20Object%20Reference) (FrameDeltaTime :type Float%20%28single-precision%29) (EventReference :type Anim%20Notify%20Event%20Reference%20Structure%20%28by%20ref%29))
  (bind _returnvalue (Components|GetOwner MeshComp))
  (bind _outactors (Actor|GetAllActorswithTag MeshComp "PlayerCompanion"))
  (bind _ascompanion_movement_interface (Utilities|Casting|CastToCompanionMovementInterface (Utilities|Array|Get(acopy) _outactors)))
  (bind _aspawn (Utilities|Casting|CastToPawn _returnvalue))
  (bind _asplayer_controller (Utilities|Casting|CastToPlayerController (Pawn|GetController _aspawn)))
  (bind (_outhit _returnvalue_2) (Collision|LineTraceByChannel _asplayer_controller (Pawn|GetPlayerViewPoint _asplayer_controller) (+ (Pawn|GetPlayerViewPoint _asplayer_controller) (* (Math|Vector|GetForwardVector (.Rotation (Pawn|GetPlayerViewPoint _asplayer_controller))) 5e+03)) "TraceTypeQuery1" false (Utilities|Array|MakeArray _returnvalue)))
  (if _returnvalue_2
    (bind _output_get (Variables|/CosmeticCompanions/Assets/Quadruped/FloppyNib/Animation/Emotes/FlobbyNib_LaserEmote/FloppyNib_Emote_Component.FloppyNib_Emote_Component_C|SetLaserTargetLocation (.Location (Collision|BreakHitResult _outhit)) (Actor|GetComponentbyClass (Utilities|Array|Get(acopy) _outactors) "/CosmeticCompanions/Assets/Quadruped/FloppyNib/Animation/Emotes/FlobbyNib_LaserEmote/FloppyNib_Emote_Component.FloppyNib_Emote_Component_C")))
    (if (Utilities|IsValid (Variables|/CosmeticCompanions/Assets/Quadruped/FloppyNib/Animation/Emotes/FlobbyNib_LaserEmote/FloppyNib_Emote_Component.FloppyNib_Emote_Component_C|GetLaserDotDecal (Actor|GetComponentbyClass (Utilities|Array|Get(acopy) _outactors) "/CosmeticCompanions/Assets/Quadruped/FloppyNib/Animation/Emotes/FlobbyNib_LaserEmote/FloppyNib_Emote_Component.FloppyNib_Emote_Component_C")) :!class /Script/Engine.KismetSystemLibrary)
      (Transformation|SetWorldLocationAndRotation (Variables|/CosmeticCompanions/Assets/Quadruped/FloppyNib/Animation/Emotes/FlobbyNib_LaserEmote/FloppyNib_Emote_Component.FloppyNib_Emote_Component_C|GetLaserDotDecal (Actor|GetComponentbyClass (Utilities|Array|Get(acopy) _outactors) "/CosmeticCompanions/Assets/Quadruped/FloppyNib/Animation/Emotes/FlobbyNib_LaserEmote/FloppyNib_Emote_Component.FloppyNib_Emote_Component_C")) _output_get (Math|Conversions|RotationFromXVector (.ImpactNormal (Collision|BreakHitResult _outhit))) false false)
      ; Random swipes when within range
      (Variables|/CosmeticCompanions/Assets/Quadruped/FloppyNib/Animation/Emotes/FlobbyNib_LaserEmote/FloppyNib_Emote_Component.FloppyNib_Emote_Component_C|SetSwipeCoolDownTimer (- (Variables|/CosmeticCompanions/Assets/Quadruped/FloppyNib/Animation/Emotes/FlobbyNib_LaserEmote/FloppyNib_Emote_Component.FloppyNib_Emote_Component_C|GetSwipeCoolDownTimer (Actor|GetComponentbyClass (Utilities|Array|Get(acopy) _outactors) "/CosmeticCompanions/Assets/Quadruped/FloppyNib/Animation/Emotes/FlobbyNib_LaserEmote/FloppyNib_Emote_Component.FloppyNib_Emote_Component_C")) FrameDeltaTime) (Actor|GetComponentbyClass (Utilities|Array|Get(acopy) _outactors) "/CosmeticCompanions/Assets/Quadruped/FloppyNib/Animation/Emotes/FlobbyNib_LaserEmote/FloppyNib_Emote_Component.FloppyNib_Emote_Component_C"))
      (if (<= (Variables|/CosmeticCompanions/Assets/Quadruped/FloppyNib/Animation/Emotes/FlobbyNib_LaserEmote/FloppyNib_Emote_Component.FloppyNib_Emote_Component_C|GetSwipeCoolDownTimer (Actor|GetComponentbyClass (Utilities|Array|Get(acopy) _outactors) "/CosmeticCompanions/Assets/Quadruped/FloppyNib/Animation/Emotes/FlobbyNib_LaserEmote/FloppyNib_Emote_Component.FloppyNib_Emote_Component_C")) 0)
        (if (not (Variables|/CosmeticCompanions/Assets/Quadruped/FloppyNib/Animation/Emotes/FlobbyNib_LaserEmote/FloppyNib_Emote_Component.FloppyNib_Emote_Component_C|GetbShouldPounce (Actor|GetComponentbyClass (Utilities|Array|Get(acopy) _outactors) "/CosmeticCompanions/Assets/Quadruped/FloppyNib/Animation/Emotes/FlobbyNib_LaserEmote/FloppyNib_Emote_Component.FloppyNib_Emote_Component_C")))
          ;; this used to let the cat swipe while moving, disconnected
          (if (Variables|/CosmeticCompanions/Assets/Quadruped/FloppyNib/Animation/Emotes/FlobbyNib_LaserEmote/FloppyNib_Emote_Component.FloppyNib_Emote_Component_C|GetbShouldCrouch (Actor|GetComponentbyClass (Utilities|Array|Get(acopy) _outactors) "/CosmeticCompanions/Assets/Quadruped/FloppyNib/Animation/Emotes/FlobbyNib_LaserEmote/FloppyNib_Emote_Component.FloppyNib_Emote_Component_C"))
            (if (> (Math|Vector|DotProduct (Transformation|GetActorRightVector (Utilities|Array|Get(acopy) _outactors)) (Math|Vector|Normalize (- _output_get (Transformation|GetActorLocation (Utilities|Array|Get(acopy) _outactors))))) 0.5)
              (if (< (Math|Random|RandomFloatinRange 0.0 1.0) 0.05)
                (Variables|/CosmeticCompanions/Assets/Quadruped/FloppyNib/Animation/Emotes/FlobbyNib_LaserEmote/FloppyNib_Emote_Component.FloppyNib_Emote_Component_C|SetbShouldSwipe true (Actor|GetComponentbyClass (Utilities|Array|Get(acopy) _outactors) "/CosmeticCompanions/Assets/Quadruped/FloppyNib/Animation/Emotes/FlobbyNib_LaserEmote/FloppyNib_Emote_Component.FloppyNib_Emote_Component_C"))
                (Variables|/CosmeticCompanions/Assets/Quadruped/FloppyNib/Animation/Emotes/FlobbyNib_LaserEmote/FloppyNib_Emote_Component.FloppyNib_Emote_Component_C|SetSwipeCoolDownTimer 2.0 (Actor|GetComponentbyClass (Utilities|Array|Get(acopy) _outactors) "/CosmeticCompanions/Assets/Quadruped/FloppyNib/Animation/Emotes/FlobbyNib_LaserEmote/FloppyNib_Emote_Component.FloppyNib_Emote_Component_C")))))))
      ; Should crouch when laser is close, 50% proc pounce when laser moves out
      (if (Variables|/CosmeticCompanions/Assets/Quadruped/FloppyNib/Animation/Emotes/FlobbyNib_LaserEmote/FloppyNib_Emote_Component.FloppyNib_Emote_Component_C|GetbShouldPounce (Actor|GetComponentbyClass (Utilities|Array|Get(acopy) _outactors) "/CosmeticCompanions/Assets/Quadruped/FloppyNib/Animation/Emotes/FlobbyNib_LaserEmote/FloppyNib_Emote_Component.FloppyNib_Emote_Component_C"))
        (bind _output_get_2 (Variables|/CosmeticCompanions/Assets/Quadruped/FloppyNib/Animation/Emotes/FlobbyNib_LaserEmote/FloppyNib_Emote_Component.FloppyNib_Emote_Component_C|SetPounceTimer (- (Variables|/CosmeticCompanions/Assets/Quadruped/FloppyNib/Animation/Emotes/FlobbyNib_LaserEmote/FloppyNib_Emote_Component.FloppyNib_Emote_Component_C|GetPounceTimer (Actor|GetComponentbyClass (Utilities|Array|Get(acopy) _outactors) "/CosmeticCompanions/Assets/Quadruped/FloppyNib/Animation/Emotes/FlobbyNib_LaserEmote/FloppyNib_Emote_Component.FloppyNib_Emote_Component_C")) FrameDeltaTime) (Actor|GetComponentbyClass (Utilities|Array|Get(acopy) _outactors) "/CosmeticCompanions/Assets/Quadruped/FloppyNib/Animation/Emotes/FlobbyNib_LaserEmote/FloppyNib_Emote_Component.FloppyNib_Emote_Component_C")))
        (if (<= _output_get_2 0)
          (Variables|/CosmeticCompanions/Assets/Quadruped/FloppyNib/Animation/Emotes/FlobbyNib_LaserEmote/FloppyNib_Emote_Component.FloppyNib_Emote_Component_C|SetbShouldPounce false (Actor|GetComponentbyClass (Utilities|Array|Get(acopy) _outactors) "/CosmeticCompanions/Assets/Quadruped/FloppyNib/Animation/Emotes/FlobbyNib_LaserEmote/FloppyNib_Emote_Component.FloppyNib_Emote_Component_C"))
          (Companion|SetCurrentGoal _ascompanion_movement_interface _output_get 8e+01 5.5e+02 false false)
          (Actor|Tick|SetActorTickEnabled (Utilities|Array|Get(acopy) _outactors) true)
          ; putting an OR gate in to prevent jitter
          (if (or (> (Math|Vector|Distance(Vector) (Transformation|GetActorLocation (Utilities|Array|Get(acopy) _outactors)) (.Location (Collision|BreakHitResult _outhit))) 3e+01) (> (Math|Vector|Distance(Vector) (Variables|/CosmeticCompanions/Assets/Quadruped/FloppyNib/Animation/Emotes/FlobbyNib_LaserEmote/FloppyNib_Emote_Component.FloppyNib_Emote_Component_C|GetLaserTargetLocation (Actor|GetComponentbyClass (Utilities|Array|Get(acopy) _outactors) "/CosmeticCompanions/Assets/Quadruped/FloppyNib/Animation/Emotes/FlobbyNib_LaserEmote/FloppyNib_Emote_Component.FloppyNib_Emote_Component_C")) (Variables|/CosmeticCompanions/Assets/Quadruped/FloppyNib/Animation/Emotes/FlobbyNib_LaserEmote/FloppyNib_Emote_Component.FloppyNib_Emote_Component_C|GetLastGoalLocation (Actor|GetComponentbyClass (Utilities|Array|Get(acopy) _outactors) "/CosmeticCompanions/Assets/Quadruped/FloppyNib/Animation/Emotes/FlobbyNib_LaserEmote/FloppyNib_Emote_Component.FloppyNib_Emote_Component_C"))) 5.0))
            (bind _output_get_3 (Variables|/CosmeticCompanions/Assets/Quadruped/FloppyNib/Animation/Emotes/FlobbyNib_LaserEmote/FloppyNib_Emote_Component.FloppyNib_Emote_Component_C|SetLastGoalLocation (.Location (Collision|BreakHitResult _outhit)) (Actor|GetComponentbyClass (Utilities|Array|Get(acopy) _outactors) "/CosmeticCompanions/Assets/Quadruped/FloppyNib/Animation/Emotes/FlobbyNib_LaserEmote/FloppyNib_Emote_Component.FloppyNib_Emote_Component_C")))
            (if (Variables|/CosmeticCompanions/Assets/Quadruped/FloppyNib/Animation/Emotes/FlobbyNib_LaserEmote/FloppyNib_Emote_Component.FloppyNib_Emote_Component_C|GetbShouldCrouch (Actor|GetComponentbyClass (Utilities|Array|Get(acopy) _outactors) "/CosmeticCompanions/Assets/Quadruped/FloppyNib/Animation/Emotes/FlobbyNib_LaserEmote/FloppyNib_Emote_Component.FloppyNib_Emote_Component_C"))
              _
              (elif (> (Math|Vector|Distance(Vector) (Variables|/CosmeticCompanions/Assets/Quadruped/FloppyNib/Animation/Emotes/FlobbyNib_LaserEmote/FloppyNib_Emote_Component.FloppyNib_Emote_Component_C|GetLaserTargetLocation (Actor|GetComponentbyClass (Utilities|Array|Get(acopy) _outactors) "/CosmeticCompanions/Assets/Quadruped/FloppyNib/Animation/Emotes/FlobbyNib_LaserEmote/FloppyNib_Emote_Component.FloppyNib_Emote_Component_C")) (Variables|/CosmeticCompanions/Assets/Quadruped/FloppyNib/Animation/Emotes/FlobbyNib_LaserEmote/FloppyNib_Emote_Component.FloppyNib_Emote_Component_C|GetLaserStartLocation (Actor|GetComponentbyClass (Utilities|Array|Get(acopy) _outactors) "/CosmeticCompanions/Assets/Quadruped/FloppyNib/Animation/Emotes/FlobbyNib_LaserEmote/FloppyNib_Emote_Component.FloppyNib_Emote_Component_C"))) 6e+02)
                _
                (else
                  (Companion|SetCurrentGoal _ascompanion_movement_interface _output_get_3 75.0 (Math|Interpolation|Ease "EaseOut" (Math|Float|NormalizetoRange (Math|Vector|Distance(Vector) (Transformation|GetActorLocation (Utilities|Array|Get(acopy) _outactors)) (.Location (Collision|BreakHitResult _outhit))) 2e+01 6e+02) 1e+02 5.5e+02 6.0) false false)))))
          (else
            ; pounce adding velocity and rotating to laser
            (Transformation|AddActorWorldOffset (Utilities|Array|Get(acopy) _outactors) (* (Math|Vector|Normalize (- _output_get (Transformation|GetActorLocation (Utilities|Array|Get(acopy) _outactors)))) (* (Variables|/CosmeticCompanions/Assets/Quadruped/FloppyNib/Animation/Emotes/FlobbyNib_LaserEmote/FloppyNib_Emote_Component.FloppyNib_Emote_Component_C|GetPounceSpeed (Actor|GetComponentbyClass (Utilities|Array|Get(acopy) _outactors) "/CosmeticCompanions/Assets/Quadruped/FloppyNib/Animation/Emotes/FlobbyNib_LaserEmote/FloppyNib_Emote_Component.FloppyNib_Emote_Component_C")) FrameDeltaTime)) false false)
            (Transformation|SetActorRotation (Utilities|Array|Get(acopy) _outactors) (Math|Interpolation|RInterpTo (Transformation|GetActorRotation (Utilities|Array|Get(acopy) _outactors)) (Math|Rotator|MakeRotator 0.0 0.0 (+ (.yaw (Math|Rotator|FindLookatRotation (Transformation|GetActorLocation (Utilities|Array|Get(acopy) _outactors)) _output_get)) -9e+01)) FrameDeltaTime 4.0) false)))
        (else
          (if (Variables|/CosmeticCompanions/Assets/Quadruped/FloppyNib/Animation/Emotes/FlobbyNib_LaserEmote/FloppyNib_Emote_Component.FloppyNib_Emote_Component_C|GetbShouldCrouch (Actor|GetComponentbyClass (Utilities|Array|Get(acopy) _outactors) "/CosmeticCompanions/Assets/Quadruped/FloppyNib/Animation/Emotes/FlobbyNib_LaserEmote/FloppyNib_Emote_Component.FloppyNib_Emote_Component_C"))
            ; cat leaves crouch when laser goes to the side
            (if (> (Math|Vector|DotProduct (Transformation|GetActorRightVector (Utilities|Array|Get(acopy) _outactors)) (Math|Vector|Normalize (- _output_get (Transformation|GetActorLocation (Utilities|Array|Get(acopy) _outactors))))) 0.5)
              (if (> (Math|Vector|Distance(Vector) (Transformation|GetActorLocation (Utilities|Array|Get(acopy) _outactors)) _output_get) 1.2e+02)
                (Actor|Tick|SetActorTickEnabled (Utilities|Array|Get(acopy) _outactors) true)
                (if (< (Math|Random|RandomFloatinRange 0.0 1.0) 0.5)
                  (Variables|/CosmeticCompanions/Assets/Quadruped/FloppyNib/Animation/Emotes/FlobbyNib_LaserEmote/FloppyNib_Emote_Component.FloppyNib_Emote_Component_C|SetPounceSpeed 3.5e+02 (Actor|GetComponentbyClass (Utilities|Array|Get(acopy) _outactors) "/CosmeticCompanions/Assets/Quadruped/FloppyNib/Animation/Emotes/FlobbyNib_LaserEmote/FloppyNib_Emote_Component.FloppyNib_Emote_Component_C"))
                  (Actor|Tick|SetActorTickEnabled (Utilities|Array|Get(acopy) _outactors) false)
                  (Variables|/CosmeticCompanions/Assets/Quadruped/FloppyNib/Animation/Emotes/FlobbyNib_LaserEmote/FloppyNib_Emote_Component.FloppyNib_Emote_Component_C|SetbShouldPounce true (Actor|GetComponentbyClass (Utilities|Array|Get(acopy) _outactors) "/CosmeticCompanions/Assets/Quadruped/FloppyNib/Animation/Emotes/FlobbyNib_LaserEmote/FloppyNib_Emote_Component.FloppyNib_Emote_Component_C"))
                  (Variables|/CosmeticCompanions/Assets/Quadruped/FloppyNib/Animation/Emotes/FlobbyNib_LaserEmote/FloppyNib_Emote_Component.FloppyNib_Emote_Component_C|SetPounceTimer 0.375 (Actor|GetComponentbyClass (Utilities|Array|Get(acopy) _outactors) "/CosmeticCompanions/Assets/Quadruped/FloppyNib/Animation/Emotes/FlobbyNib_LaserEmote/FloppyNib_Emote_Component.FloppyNib_Emote_Component_C"))
                  (Variables|/CosmeticCompanions/Assets/Quadruped/FloppyNib/Animation/Emotes/FlobbyNib_LaserEmote/FloppyNib_Emote_Component.FloppyNib_Emote_Component_C|SetbShouldCrouch false (Actor|GetComponentbyClass (Utilities|Array|Get(acopy) _outactors) "/CosmeticCompanions/Assets/Quadruped/FloppyNib/Animation/Emotes/FlobbyNib_LaserEmote/FloppyNib_Emote_Component.FloppyNib_Emote_Component_C"))
                  (Variables|/CosmeticCompanions/Assets/Quadruped/FloppyNib/Animation/Emotes/FlobbyNib_LaserEmote/FloppyNib_Emote_Component.FloppyNib_Emote_Component_C|SetCrouchTimer 0.0 (Actor|GetComponentbyClass (Utilities|Array|Get(acopy) _outactors) "/CosmeticCompanions/Assets/Quadruped/FloppyNib/Animation/Emotes/FlobbyNib_LaserEmote/FloppyNib_Emote_Component.FloppyNib_Emote_Component_C"))
                  (else
                    (Variables|/CosmeticCompanions/Assets/Quadruped/FloppyNib/Animation/Emotes/FlobbyNib_LaserEmote/FloppyNib_Emote_Component.FloppyNib_Emote_Component_C|SetbShouldCrouch false (Actor|GetComponentbyClass (Utilities|Array|Get(acopy) _outactors) "/CosmeticCompanions/Assets/Quadruped/FloppyNib/Animation/Emotes/FlobbyNib_LaserEmote/FloppyNib_Emote_Component.FloppyNib_Emote_Component_C"))
                    (Variables|/CosmeticCompanions/Assets/Quadruped/FloppyNib/Animation/Emotes/FlobbyNib_LaserEmote/FloppyNib_Emote_Component.FloppyNib_Emote_Component_C|SetbShouldPounce false (Actor|GetComponentbyClass (Utilities|Array|Get(acopy) _outactors) "/CosmeticCompanions/Assets/Quadruped/FloppyNib/Animation/Emotes/FlobbyNib_LaserEmote/FloppyNib_Emote_Component.FloppyNib_Emote_Component_C"))
                    (Actor|Tick|SetActorTickEnabled (Utilities|Array|Get(acopy) _outactors) true)
                    (Variables|/CosmeticCompanions/Assets/Quadruped/FloppyNib/Animation/Emotes/FlobbyNib_LaserEmote/FloppyNib_Emote_Component.FloppyNib_Emote_Component_C|SetCrouchTimer 0.0 (Actor|GetComponentbyClass (Utilities|Array|Get(acopy) _outactors) "/CosmeticCompanions/Assets/Quadruped/FloppyNib/Animation/Emotes/FlobbyNib_LaserEmote/FloppyNib_Emote_Component.FloppyNib_Emote_Component_C")))))
              (return false))
            (else
              (if (and (< (Math|Vector|Distance(Vector) (Transformation|GetActorLocation (Utilities|Array|Get(acopy) _outactors)) _output_get) 2e+02) (< (Math|Vector|VectorLength (Transformation|GetVelocity (Utilities|Array|Get(acopy) _outactors))) 5e+01))
                ;; cat only crouches when laser is directly in front of it
                (if (> (Math|Vector|DotProduct (Transformation|GetActorRightVector (Utilities|Array|Get(acopy) _outactors)) (Math|Vector|Normalize (- _output_get (Transformation|GetActorLocation (Utilities|Array|Get(acopy) _outactors))))) 0.5)
                  (bind _output_get_4 (Variables|/CosmeticCompanions/Assets/Quadruped/FloppyNib/Animation/Emotes/FlobbyNib_LaserEmote/FloppyNib_Emote_Component.FloppyNib_Emote_Component_C|SetCrouchTimer (+ (Variables|/CosmeticCompanions/Assets/Quadruped/FloppyNib/Animation/Emotes/FlobbyNib_LaserEmote/FloppyNib_Emote_Component.FloppyNib_Emote_Component_C|GetCrouchTimer (Actor|GetComponentbyClass (Utilities|Array|Get(acopy) _outactors) "/CosmeticCompanions/Assets/Quadruped/FloppyNib/Animation/Emotes/FlobbyNib_LaserEmote/FloppyNib_Emote_Component.FloppyNib_Emote_Component_C")) FrameDeltaTime) (Actor|GetComponentbyClass (Utilities|Array|Get(acopy) _outactors) "/CosmeticCompanions/Assets/Quadruped/FloppyNib/Animation/Emotes/FlobbyNib_LaserEmote/FloppyNib_Emote_Component.FloppyNib_Emote_Component_C")))
                  (if (> _output_get_4 0.2)
                    (LocalVariables|SetCrouchAnchorLocation (Transformation|GetActorLocation (Utilities|Array|Get(acopy) _outactors)))
                    (Variables|/CosmeticCompanions/Assets/Quadruped/FloppyNib/Animation/Emotes/FlobbyNib_LaserEmote/FloppyNib_Emote_Component.FloppyNib_Emote_Component_C|SetbShouldCrouch true (Actor|GetComponentbyClass (Utilities|Array|Get(acopy) _outactors) "/CosmeticCompanions/Assets/Quadruped/FloppyNib/Animation/Emotes/FlobbyNib_LaserEmote/FloppyNib_Emote_Component.FloppyNib_Emote_Component_C"))
                    (Variables|/CosmeticCompanions/Assets/Quadruped/FloppyNib/Animation/Emotes/FlobbyNib_LaserEmote/FloppyNib_Emote_Component.FloppyNib_Emote_Component_C|SetSwipeCoolDownTimer 2.0 (Actor|GetComponentbyClass (Utilities|Array|Get(acopy) _outactors) "/CosmeticCompanions/Assets/Quadruped/FloppyNib/Animation/Emotes/FlobbyNib_LaserEmote/FloppyNib_Emote_Component.FloppyNib_Emote_Component_C"))
                    (Actor|Tick|SetActorTickEnabled (Utilities|Array|Get(acopy) _outactors) false)
                    (else
                      _))
                  (else
                    (Variables|/CosmeticCompanions/Assets/Quadruped/FloppyNib/Animation/Emotes/FlobbyNib_LaserEmote/FloppyNib_Emote_Component.FloppyNib_Emote_Component_C|SetbShouldCrouch false (Actor|GetComponentbyClass (Utilities|Array|Get(acopy) _outactors) "/CosmeticCompanions/Assets/Quadruped/FloppyNib/Animation/Emotes/FlobbyNib_LaserEmote/FloppyNib_Emote_Component.FloppyNib_Emote_Component_C"))
                    (Variables|/CosmeticCompanions/Assets/Quadruped/FloppyNib/Animation/Emotes/FlobbyNib_LaserEmote/FloppyNib_Emote_Component.FloppyNib_Emote_Component_C|SetCrouchTimer 0.0 (Actor|GetComponentbyClass (Utilities|Array|Get(acopy) _outactors) "/CosmeticCompanions/Assets/Quadruped/FloppyNib/Animation/Emotes/FlobbyNib_LaserEmote/FloppyNib_Emote_Component.FloppyNib_Emote_Component_C"))))
                (elif (> (Math|Vector|VectorLength (Transformation|GetVelocity (Utilities|Array|Get(acopy) _outactors))) 5e+01)
                  ; randomly procs a pounce when cat is moving and within range
                  (if (< (Math|Vector|Distance(Vector) (Transformation|GetActorLocation (Utilities|Array|Get(acopy) _outactors)) _output_get) 1e+02)
                    (if (> (Math|Vector|DotProduct (Transformation|GetActorRightVector (Utilities|Array|Get(acopy) _outactors)) (Math|Vector|Normalize (- _output_get (Transformation|GetActorLocation (Utilities|Array|Get(acopy) _outactors))))) 0.5)
                      (if (not (Variables|/CosmeticCompanions/Assets/Quadruped/FloppyNib/Animation/Emotes/FlobbyNib_LaserEmote/FloppyNib_Emote_Component.FloppyNib_Emote_Component_C|GetbApproachPounceRolled (Actor|GetComponentbyClass (Utilities|Array|Get(acopy) _outactors) "/CosmeticCompanions/Assets/Quadruped/FloppyNib/Animation/Emotes/FlobbyNib_LaserEmote/FloppyNib_Emote_Component.FloppyNib_Emote_Component_C")))
                        (Variables|/CosmeticCompanions/Assets/Quadruped/FloppyNib/Animation/Emotes/FlobbyNib_LaserEmote/FloppyNib_Emote_Component.FloppyNib_Emote_Component_C|SetbApproachPounceRolled true (Actor|GetComponentbyClass (Utilities|Array|Get(acopy) _outactors) "/CosmeticCompanions/Assets/Quadruped/FloppyNib/Animation/Emotes/FlobbyNib_LaserEmote/FloppyNib_Emote_Component.FloppyNib_Emote_Component_C"))
                        (if (< (Math|Random|RandomFloatinRange 0.0 1.0) 0.3)
                          (Variables|/CosmeticCompanions/Assets/Quadruped/FloppyNib/Animation/Emotes/FlobbyNib_LaserEmote/FloppyNib_Emote_Component.FloppyNib_Emote_Component_C|SetPounceSpeed 4e+02 (Actor|GetComponentbyClass (Utilities|Array|Get(acopy) _outactors) "/CosmeticCompanions/Assets/Quadruped/FloppyNib/Animation/Emotes/FlobbyNib_LaserEmote/FloppyNib_Emote_Component.FloppyNib_Emote_Component_C"))
                          (Variables|/CosmeticCompanions/Assets/Quadruped/FloppyNib/Animation/Emotes/FlobbyNib_LaserEmote/FloppyNib_Emote_Component.FloppyNib_Emote_Component_C|SetbShouldPounce true (Actor|GetComponentbyClass (Utilities|Array|Get(acopy) _outactors) "/CosmeticCompanions/Assets/Quadruped/FloppyNib/Animation/Emotes/FlobbyNib_LaserEmote/FloppyNib_Emote_Component.FloppyNib_Emote_Component_C"))
                          (Variables|/CosmeticCompanions/Assets/Quadruped/FloppyNib/Animation/Emotes/FlobbyNib_LaserEmote/FloppyNib_Emote_Component.FloppyNib_Emote_Component_C|SetPounceTimer 0.375 (Actor|GetComponentbyClass (Utilities|Array|Get(acopy) _outactors) "/CosmeticCompanions/Assets/Quadruped/FloppyNib/Animation/Emotes/FlobbyNib_LaserEmote/FloppyNib_Emote_Component.FloppyNib_Emote_Component_C"))
                          (Actor|Tick|SetActorTickEnabled (Utilities|Array|Get(acopy) _outactors) false)
                          (else
                            _))))
                    (else
                      (Variables|/CosmeticCompanions/Assets/Quadruped/FloppyNib/Animation/Emotes/FlobbyNib_LaserEmote/FloppyNib_Emote_Component.FloppyNib_Emote_Component_C|SetbApproachPounceRolled false (Actor|GetComponentbyClass (Utilities|Array|Get(acopy) _outactors) "/CosmeticCompanions/Assets/Quadruped/FloppyNib/Animation/Emotes/FlobbyNib_LaserEmote/FloppyNib_Emote_Component.FloppyNib_Emote_Component_C"))))
                  (Variables|/CosmeticCompanions/Assets/Quadruped/FloppyNib/Animation/Emotes/FlobbyNib_LaserEmote/FloppyNib_Emote_Component.FloppyNib_Emote_Component_C|SetCrouchTimer 0.0 (Actor|GetComponentbyClass (Utilities|Array|Get(acopy) _outactors) "/CosmeticCompanions/Assets/Quadruped/FloppyNib/Animation/Emotes/FlobbyNib_LaserEmote/FloppyNib_Emote_Component.FloppyNib_Emote_Component_C"))
                  (else
                    (Variables|/CosmeticCompanions/Assets/Quadruped/FloppyNib/Animation/Emotes/FlobbyNib_LaserEmote/FloppyNib_Emote_Component.FloppyNib_Emote_Component_C|SetCrouchTimer 0.0 (Actor|GetComponentbyClass (Utilities|Array|Get(acopy) _outactors) "/CosmeticCompanions/Assets/Quadruped/FloppyNib/Animation/Emotes/FlobbyNib_LaserEmote/FloppyNib_Emote_Component.FloppyNib_Emote_Component_C"))))))))))))
