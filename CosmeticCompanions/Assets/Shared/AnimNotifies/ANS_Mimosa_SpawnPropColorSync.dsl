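; Anim notify state that color-syncs a spawned cosmetic companion prop to its owner's materials via CustomizePropMeshComponent.
; It finds skeletal meshes on the attach parent's owner whose display names contain a search string.
; For each slot mapping, it creates a dynamic material on the prop, either from the whole source material or by copying mapped vector parameters.
; Parameter copying requires a dynamic source material; debug prints report matches and cast failures.
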(fn CustomizePropMeshComponent
  (params (MeshComponent :type Mesh%20Component%20Object%20Reference))
  (bind _asskeletal_mesh_component (Utilities|Casting|CastToSkeletalMeshComponent (Transformation|GetAttachParent MeshComponent)))
  (for _array_element (Actor|GetComponentsByClass (Components|GetOwner _asskeletal_mesh_component) "/Script/Engine.SkeletalMeshComponent")
    (if (Utilities|String|Contains (Utilities|GetDisplayName _array_element) (Variables|Default|GetComponentSearchString))
      (Development|PrintString 0 (Utilities|GetDisplayName _array_element))
      (for _array_element_2 (Variables|Default|GetMaterialSlotMappings)
        (if (Variables|Default|GetbCopyEntireMaterial)
          (Rendering|Material|CreateDynamicMaterialInstance MeshComponent (.TargetElementIndex_4_FBBB3E864931FD5B2DA1CA8F9DEAEBFB (Utilities|Struct|BreakSMaterialSlotMapping _array_element_2)) (Rendering|Material|GetMaterial _array_element (Utilities|Struct|BreakSMaterialSlotMapping _array_element_2)) "" :!class /Script/Engine.PrimitiveComponent)
          (else
            (bind _asmaterial_instance_dynamic (Utilities|Casting|CastToMaterialInstanceDynamic (Rendering|Material|GetMaterial _array_element (Utilities|Struct|BreakSMaterialSlotMapping _array_element_2)))
              (:then
                (bind _returnvalue (Rendering|Material|CreateDynamicMaterialInstance MeshComponent (.TargetElementIndex_4_FBBB3E864931FD5B2DA1CA8F9DEAEBFB (Utilities|Struct|BreakSMaterialSlotMapping _array_element_2)) 0 "" :!class /Script/Engine.PrimitiveComponent))
                (for _array_element_3 (.ParameterMappings_8_094ED4D94509CEDDCEC89083927A567A (Utilities|Struct|BreakSMaterialSlotMapping _array_element_2))
                  (bind _returnvalue_2 (Rendering|Material|GetVectorParameterValue _asmaterial_instance_dynamic (Utilities|Struct|BreakSParameterMapping _array_element_3) :!class /Script/Engine.MaterialInstanceDynamic))
                  (Rendering|Material|SetVectorParameterValue _returnvalue (.TargetParameterName_4_2F9C0B67407500F5D3B33AAC629EE1DE (Utilities|Struct|BreakSParameterMapping _array_element_3)) _returnvalue_2 :!class /Script/Engine.MaterialInstanceDynamic)))
              (:CastFailed
                (Development|PrintString 0 "cast to mid fail")))))))))
