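; Spawn-prop anim notify for cosmetic companions that swaps the materials on the prop mesh it spawns.
; CustomizePropMeshComponent reads the notify's material override array and passes it to a helper after casting the component to a skeletal mesh, falling back to static mesh.
; The helper walks the mesh's material slots and replaces each one that has a matching index in the array, leaving other slots untouched.
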(fn CustomizePropMeshComponent
  (params (MeshComponent :type Mesh%20Component%20Object%20Reference))
  (bind _materialoverrides (Variables|Default|GetMaterialOverrides))
  (LocalVariables|SetLocalMeshComponent MeshComponent)
  (bind _asskeletal_mesh_component (Utilities|Casting|CastToSkeletalMeshComponent (LocalVariables|GetLocalMeshComponent))
    (:then
      (CallFunction|SetMaterialOverrides _asskeletal_mesh_component _materialoverrides))
    (:CastFailed
      (bind _asstatic_mesh_component (Utilities|Casting|CastToStaticMeshComponent (LocalVariables|GetLocalMeshComponent)))
      (CallFunction|SetMaterialOverrides _asstatic_mesh_component _materialoverrides))))

(fn SetMaterialOverrides
  (params (Mesh%20Component :type Mesh%20Component%20Object%20Reference) (Materials :type Array%20of%20Material%20Interface%20Object%20References))
  (for (_ _array_index) (Rendering|Material|GetMaterials Mesh%20Component)
    (if (Utilities|Array|IsValidIndex Materials _array_index)
      (Rendering|Material|SetMaterial Mesh%20Component _array_index (Utilities|Array|Get(acopy) Materials _array_index)))))
