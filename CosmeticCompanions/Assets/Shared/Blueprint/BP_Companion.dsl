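; BP_Companion is a cosmetic companion Blueprint with debug toggles that show wireframe capsule meshes on its skeletal mesh.
; Toggle Capsule creates or destroys a fixed-transform capsule attached to the root socket.
; Toggle Accessory does the same using the Accessory Capsule Tweak transform and Accessory Attach Socket variables.
; Both add components at runtime and use the stored reference's validity as toggle state; other events are empty.
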(event Custom|Toggle%20Capsule
  ;[0.15 0.15 0.15 0] Toggle Capsule
  (Utilities|IsValid (Variables|Default|GetCapsule%20Component)
    (:Is%20Valid
      ;;[0.15 0 0 0.5] Destroy Capsule
      (Components|DestroyComponent (Variables|Default|GetCapsule%20Component)))
    (:Is%20Not%20Valid
      ;;[0 0.15 0 0.5] Create Capsule
      (bind _returnvalue (Game|AddComponentbyClass 0 "/Script/Engine.StaticMeshComponent" false (Math|Transform|MakeTransform "0.000000,0.000000,43.717771" "0, 0, 0" "4.979554,4.979554,5.318806")))
      (Variables|Default|SetCapsule%20Component _returnvalue)
      (Components|StaticMesh|SetStaticMesh _returnvalue "/CosmeticCompanions/Assets/Shared/Meshes/SM_capsule.SM_capsule")
      (Rendering|Material|SetMaterial _returnvalue 0 "/Engine/EngineDebugMaterials/WireframeMaterial.WireframeMaterial")
      (Transformation|AttachComponentToComponent _returnvalue (Variables|Default|GetSkeletalMesh) "root" "KeepRelative" "KeepRelative" "KeepRelative" true))))

(event Custom|Toggle%20Accessory
  ;[0.15 0.15 0.15 0] Toggle Accessory
  (Utilities|IsValid (Variables|Default|GetAccessory%20Component)
    (:Is%20Valid
      ;;[0.15 0 0 0.5] Destroy Accessory
      (Components|DestroyComponent (Variables|Default|GetAccessory%20Component)))
    (:Is%20Not%20Valid
      ;;[0 0.15 0 0.5] Create Accessory
      (bind _returnvalue (Game|AddComponentbyClass 0 "/Script/Engine.StaticMeshComponent" false (Variables|Default|GetAccessory%20Capsule%20Tweak)))
      (Variables|Default|SetAccessory%20Component _returnvalue)
      (Components|StaticMesh|SetStaticMesh _returnvalue "/CosmeticCompanions/Assets/Shared/Meshes/SM_capsule.SM_capsule")
      (Rendering|Material|SetMaterial _returnvalue 0 "/Engine/EngineDebugMaterials/WireframeMaterial.WireframeMaterial")
      (Transformation|AttachComponentToComponent _returnvalue (Variables|Default|GetSkeletalMesh) (Variables|Default|GetAccessory%20Attach%20Socket) "KeepRelative" "KeepRelative" "KeepRelative" true))))

(event EventBeginPlay)

(event Collision|EventActorBeginOverlap
  (params (OtherActor :type Actor%20Object%20Reference)))

(event EventTick
  (params (DeltaSeconds :type Float%20%28single-precision%29)))

(fn ConstructionScript)
