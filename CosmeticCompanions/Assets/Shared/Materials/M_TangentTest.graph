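# M_TangentTest is a Default Lit material for skeletal and deformed meshes, apparently for testing anisotropic tangents.
# It samples a default anisotropy texture, splits the RGB, negates red and rebuilds the vector, mirroring its X direction.
# A UV Editor color grid is sampled with UVs tiled by scalar B, alongside a standalone Aniso scalar.
# The file does not show which material outputs these branches feed.
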Material /CosmeticCompanions/Assets/Shared/Materials/M_TangentTest
graph {
 expressions {
  MaterialExpressionConvert_0 {
   class=MaterialExpressionConvert
   inputs [None {from=MaterialExpressionTextureSample_0, output=RGB}]
   properties {
    nodeName=Break Float3
    convertInputs [{
      type=Vector3
      expressionInput {
       expression=&self:MaterialExpressionTextureSample_0
       inputName=None
       mask=1
       maskA=0
       maskB=1
       maskG=1
       maskR=1
       outputIndex=0
      }
     }]
    convertOutputs [{type=Scalar}, {type=Scalar}, {type=Scalar}]
   }
  }
  MaterialExpressionConvert_1 {
   class=MaterialExpressionConvert
   inputs [
    None {from=MaterialExpressionMultiply_0}
    None {from=MaterialExpressionConvert_0}
    None {from=MaterialExpressionConvert_0}
   ]
   properties {
    nodeName=Make Float3
    convertInputs [
     {
      type=Scalar
      expressionInput {
       expression=&self:MaterialExpressionMultiply_0
       inputName=None
       mask=0
       maskA=0
       maskB=0
       maskG=0
       maskR=0
       outputIndex=0
      }
     }
     {
      type=Scalar
      expressionInput {
       expression=&self:MaterialExpressionConvert_0
       inputName=None
       mask=0
       maskA=0
       maskB=0
       maskG=0
       maskR=0
       outputIndex=1
      }
     }
     {
      type=Scalar
      expressionInput {
       expression=&self:MaterialExpressionConvert_0
       inputName=None
       mask=0
       maskA=0
       maskB=0
       maskG=0
       maskR=0
       outputIndex=2
      }
     }
    ]
    convertOutputs [{type=Vector3}]
   }
  }
  MaterialExpressionMultiply_0 {
   class=MaterialExpressionMultiply
   inputs [
    A {from=MaterialExpressionConvert_0}
    B {from=""}
   ]
   properties {constB=-1}
  }
  MaterialExpressionMultiply_1 {
   class=MaterialExpressionMultiply
   inputs [
    A {from=MaterialExpressionTextureCoordinate_0}
    B {from=MaterialExpressionScalarParameter_1}
   ]
  }
  MaterialExpressionScalarParameter_0 {class=MaterialExpressionScalarParameter, properties {defaultValue=0.1, parameterName=Aniso}}
  MaterialExpressionScalarParameter_1 {class=MaterialExpressionScalarParameter, properties {defaultValue=5, parameterName=B}}
  MaterialExpressionTextureCoordinate_0 {class=MaterialExpressionTextureCoordinate}
  MaterialExpressionTextureSample_0 {
   class=MaterialExpressionTextureSample
   inputs [
    UVs {from=""}
    Tex {from=""}
    Apply View MipBias {from=""}
   ]
   properties {samplerType=SAMPLERTYPE_LinearColor, texture=/Game/Global/Textures/T_Default_Aniso}
  }
  MaterialExpressionTextureSample_1 {
   class=MaterialExpressionTextureSample
   inputs [
    UVs {from=MaterialExpressionMultiply_1}
    Tex {from=""}
    Apply View MipBias {from=""}
   ]
   properties {texture=/UVEditor/Textures/UVEditorColorGrid_Color}
  }
 }
 properties {bAutomaticallySetUsageInEditor=false, bUsedWithMeshDeformer=true, bUsedWithSkeletalMesh=true, usedShadingModels=Default Lit}
}
